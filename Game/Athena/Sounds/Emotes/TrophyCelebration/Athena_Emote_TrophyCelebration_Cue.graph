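# Sound cue for the Trophy Celebration emote music in Athena, routed to a first-person emote music sound class.
# A mixer plays the intro at once, slightly boosted, alongside a looping wave held back by a fixed delay, forming an intro-to-loop sequence.
# It uses the shared first-person emote music concurrency set and has a send to the music audio analysis submix.
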SoundCue /Game/Athena/Sounds/Emotes/TrophyCelebration/Athena_Emote_TrophyCelebration_Cue
root=Mixer_0
nodes [
 {id=Delay_4, children [WavePlayer_8], properties {delayMax=5.837, delayMin=5.837}}
 {id=Mixer_0, children [WavePlayer_7,Delay_4], properties {inputVolume [1.2,1]}}
 {id=WavePlayer_7, properties {soundWaveAssetPtr=&/Emote_TrophyCelebration_Intro}}
 {id=WavePlayer_8, properties {bLooping=true, soundWaveAssetPtr=&/Emote_TrophyCelebration_Loop}}
]
properties {
 maxDistance=2097152
 soundClassObject=/Game/Sounds/FortSoundClassesAndModes/Fort_Music_Emote_1P_PSM
 volumeMultiplier=1.6
 concurrencyOverrides {maxCount=3, resolutionRule=StopOldest}
 concurrencySet [/Game/Sounds/Concurrency_Rules/Athena_EmoteMusic_1P_Conc]
 soundSubmixSends [{minSendLevel=0, soundSubmix=&Fortnite_AudioAnalysis_Music}]
}
